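# Translucent, two-sided lit material for steam on Niagara mesh particles in the Peaceful Poem Remix backpack effect.
# Two scrolling inky-smoke samples, a V gradient and double-sided Fresnel build opacity, scaled by particle alpha and vertex red, with depth fade on High and Epic quality.
# Panning noise along vertex normals and DynamicParameter drag, both weighted by an inverted V gradient, form a displacement, alongside a CameraOffset function.
# Particle color feeds an emissive-adjustment function.
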Material /BRCosmetics/Accessories/FORT_Backpacks/Backpack_M_MED_PeacefulPoemRemix/FX/M_FX_Backpack_PeacefulPoemRemix_Steam
graph {
 properties {
  bUsedWithNiagaraMeshParticles=true
  blendMode=BLEND_Translucent
  pixelDepthOffsetMode=PDOM_Legacy
  translucencyLightingMode=TLM_VolumetricPerVertexNonDirectional
  twoSided=true
  usedShadingModels=Default Lit
 }
 nodes [
  n1 = TextureCoordinate() {uTiling=0.1, vTiling=0.1}
  n2 = Multiply(A=TextureSample(UVs=Panner(Coordinate=@n1) {speedY=0.1}) {texture=&T_Inky_Smoke_Tile}.R) {constB=2}
  n3 = Multiply(A=Multiply(A=Subtract(A=@n2)) {constB=-10}, B=VertexNormalWS())
  n4 = ComponentMask(None=TextureCoordinate()) {g=true}
  n5 = Power(Base=OneMinus(None=Saturate(None=@n4)))
  n6 = DynamicParameter() {defaultValue=<rgba 1 1 50 10>, paramNames [DragX,DragY,DragZ,Param4]}
  n7 = AppendVector(A=@n6.DragX, B=@n6.DragX)
  Add(A=Add(B=Multiply(A=@n3, B=@n5)), B=Multiply(A=@n5, B=AppendVector(A=@n7, B=@n6.DragZ)))
  Constant() {r=20}
  n8 = ParticleColor()
  MaterialFunctionCall(Emissive=@n8.RGB) {materialFunction=&MF_FX_EmissiveAdjustments}
  MaterialFunctionCall(Offset Amount=ScalarParameter() {parameterName=CameraOffset}) {materialFunction=/Engine/Functions/Engine_MaterialFunctions02/WorldPositionOffset/CameraOffset}
  n9 = ComponentMask(None=TextureCoordinate()) {g=true}
  n10 = Multiply(A=Time()) {constB=2}
  n11 = TextureSample(UVs=Panner(Coordinate=TextureCoordinate() {uTiling=3, vTiling=0.4}, Time=@n10) {bFractionalPart=true, speedY=0.1}) {texture=&T_Inky_Smoke_Tile}
  n12 = TextureSample(UVs=Panner(Coordinate=TextureCoordinate() {uTiling=2, vTiling=0.5}, Time=@n10) {bFractionalPart=true, speedY=0.125}) {texture=&T_Inky_Smoke_Tile}
  n13 = Add(A=@n11.R, B=@n12.R)
  n14 = Multiply(
   A=Multiply(A=Power(Base=@n9) {constExponent=3}, B=Divide(A=@n13))
   B=MaterialFunctionCall(Exponent=ScalarParameter() {defaultValue=1.5, parameterName=Fresnel}) {materialFunction=&MF_Fresnel_FX}.Fresnel Double Sided
  )
  n15 = DepthFade(Opacity=@n14) {fadeDistanceDefault=3}
  Multiply(A=Multiply(A=@n8.A, B=QualitySwitch(Default=@n14, High=@n15, Epic=@n15)), B=VertexColor().R)
  Multiply(A=@n11.R, B=@n12.R)
  VertexColor()
 ]
}
