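# Idle Niagara effect for the PeacefulPoemRemix backpack, with a steam plume and hanging syrup drips.
# MeshSteam smooths a trailing position and feeds its lag to the material, fading by running and skydiving state.
# SyrupDrips001 spawns stretching meshes on skeletal sockets, swung by pendulum physics until late-life release, with front-end and time-of-day tinting.
# Droplets and the original SyrupDrips emitter are disabled.
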NiagaraSystem /BRCosmetics/Accessories/FORT_Backpacks/Backpack_M_MED_PeacefulPoemRemix/FX/NS_Backpack_PeacefulPoemRemix_Idle
dependencies {
 usedDataInterfaces [
  &NiagaraDataInterfaceCurve
  &NiagaraDataInterfaceVector2DCurve
  &NiagaraDataInterfaceArrayFloat
  &NiagaraDataInterfaceSkeletalMesh
 ]
 usedDynamicInputs [&VectorLength]
 usedModules [
  &self:SystemSpawnScript.SystemScriptSource.SystemScriptGraph.NiagaraNodeAssignment_0.SetVariables_#1
  &self:SystemSpawnScript.SystemScriptSource.SystemScriptGraph.NiagaraNodeAssignment_2.SetVariables_#2
  &SystemState
  &self:SystemSpawnScript.SystemScriptSource.SystemScriptGraph.NiagaraNodeAssignment_1.SetVariables_#3
 ]
}
emitters {
 Droplets {
  bEnabled=false
  emitterUpdateScript [
   &EmitterState(Life Cycle Mode=::NewEnumerator1, Loop Duration=2, UseLoopDelay=false)
   &SpawnRate(Spawn Probability=0.3, SpawnRate=3, Use Spawn Probability=true)
  ]
  particleSpawnScript [
   &InitializeParticle(
    Color=<rgba 1 0.4921 0 1>
    Lifetime Max=1
    Lifetime Mode=::NewEnumerator1
    Mass Mode=::NewEnumerator3
    Sprite Rotation Mode=::NewEnumerator0
    Sprite Size Mode=::NewEnumerator1
    Uniform Sprite Size Max=2.5
    Uniform Sprite Size Min=1
    UsePositionOffset=false
    Write Material Random=true
   )
   &ShapeLocation(MaskSpawnGroup=false, Override Seed=false, Sphere Radius=8, UseOwnerScale=true)
   &AddVelocity(
    Cone Angle=32
    Cone Axis=<0 0 1>
    Distribution Along Cone Axis=0.25
    Module.UseConeFalloff=true
    Module.UseVelDistribution=true
    Override Seed=false
    Speed Falloff From Cone Axis=0.5
    Velocity Mode=::NewEnumerator2
    Velocity Speed=&RandomRangeFloat(Minimum=500, Maximum=850)
   ) {enabled=false}
   &SkeletalMeshLocation(Bone Sampling Mode=::NewEnumerator4, MaskSpawnGroup=false, Skeletal Mesh=$System.Skeletal Mesh)
  ]
  particleUpdateScript [
   &ParticleState()
   &SkeletalMeshLocation(
    Bone Sampling Mode=::NewEnumerator4
    "Inherit Velocity (Scale)"=0
    MaskSpawnGroup=false
    Orientation Sampling=::NewEnumerator1
    Position Sampling=::NewEnumerator2
    Skeletal Mesh=$System.Skeletal Mesh
    Velocity Sampling=::NewEnumerator0
    VelocitySpeedLimit=false
   ) {module=SkeletalMeshLocation001}
   &ScaleSpriteSize(
    Scale Factor=&Vector2DFromCurve(
     Vector2Curve=DataInterface() {propertyValues {
       XCurve {keys [<at 0 0 RCIM_Cubic>, {arriveTangent=0.5, interpMode=RCIM_Cubic, leaveTangent=0.5, time=0.3, value=1}, <at 1 0.5 RCIM_Cubic>]}
       YCurve {keys [<at 0 0>, {arriveTangent=4, interpMode=RCIM_Cubic, leaveTangent=4, time=0.3, value=1}, <at 1 4>]}
      }}
     ScaleCurve=<1 1>
    )
    Scale Sprite Size Mode=::NewEnumerator2
   )
   /SetVariables(Particles.GravityScale:NiagaraFloat=&ClampFloat(Float=&FloatFromCurve(FloatCurve=DataInterface() {propertyValues {Curve {keys [
          <at 0 0 RCIM_Cubic>
          {arriveTangent=0.1297, interpMode=RCIM_Cubic, leaveTangent=0.1297, tangentMode=RCTM_User, time=0.25, value=0}
          <at 1 1 RCIM_Cubic>
         ]}}}))) {isSetParameters=true}
   &GravityForce(Gravity=&Multiply_VectorByFloat(Vector=<0 0 -567.6419>, Float=$Particles.GravityScale))
   &Drag(Drag=0.9227, Use Linear Drag=true, Use Rotational Drag=false)
   &ScaleColor(
    Scale Alpha=&FloatFromCurve(FloatCurve=DataInterface() {propertyValues {Curve {keys [<at 0.7629 1.0113 RCIM_Cubic>,<at 1.0011 0.002 RCIM_Cubic>]}}})
    ScaleA=true
    ScaleRGB=true
   )
   &InheritSourceMovement(
    Applied Position Scale=&VectorFromFloat(Value=&FloatFromCurve(FloatCurve=DataInterface() {propertyValues {Curve {keys [<at 0.4 1>,<at 1 0.75>]}}}))
    ApplyForce=false
    ApplyPosition=true
    ApplyPositionFalloff=false
   )
   /SetVariables(Particles.ParticleAnimatedPos:NiagaraPosition=&Lerp_Vector(
     A=$StackContext.SkeletalMeshLocation001.SampledPosition
     B=Linked() {linkedVariable {name=Particles.Position, type=NiagaraPosition}}
     Alpha=&FloatFromCurve(FloatCurve=DataInterface() {propertyValues {Curve {keys [
          {arriveTangent=4.1785, interpMode=RCIM_Cubic, leaveTangent=4.1785, tangentMode=RCTM_User, time=0.4, value=0}
          {arriveTangent=1.1179, interpMode=RCIM_Cubic, leaveTangent=1.1179, tangentMode=RCTM_User, time=0.6, value=0.625}
          {arriveTangent=0.0224, interpMode=RCIM_Cubic, leaveTangent=0.0224, tangentMode=RCTM_User, time=1, value=1}
         ]}}})
    )) {isSetParameters=true}
   &InheritVelocity(Inherited Velocity Amount Scale=&VectorFromFloat(Value=0.1), Speed Limit=true, UseSpeedThreshold=false) {enabled=false}
   &SolveForcesAndVelocity(Clamp Velocity=false, Limit Acceleration=false)
  ]
  rendererClasses [&NiagaraSpriteRendererProperties]
  renderers [{rendererClass=&NiagaraSpriteRendererProperties}]
 }
 MeshSteam {
  emitterSpawnScript [/SetVariables(Emitter.OffsetTarget:Vector3f=<-10 0 25>) {isSetParameters=true}]
  emitterUpdateScript [
   &EmitterState(Life Cycle Mode=::NewEnumerator1, Loop Behavior=::NewEnumerator1, Loop Duration Mode=::NewEnumerator1, UseLoopDelay=false)
   &SpawnBurst_Instantaneous(Use Spawn Probability=false, UseLoopCountLimit=false)
   /SetVariables(Emitter.OffsetTarget:Vector3f=/Niagara/DynamicInputs/Lerp/SmoothLerpOverTime_Vector(
     Convergence Rate:NiagaraFloat=5
     Smooth Value:Vector3f=$Emitter.OffsetTarget
     Target Value:Vector3f=&MakeCustomVectorFromBool(Bool=$System.IsRunning, True Vector=<20 -10 0>, False Vector=<-25 -20 0>)
    )) {isSetParameters=true}
   /SetVariables(Emitter.AlphaTarget:NiagaraFloat=/Niagara/DynamicInputs/Lerp/SmoothLerpOverTime_Float(
     Convergence Rate:NiagaraFloat=5
     Smooth Value:NiagaraFloat=$Emitter.AlphaTarget
     Target Value:NiagaraFloat=&MakeCustomFloatFromBool(
      Bool=$User.IsSkydiving
      True Float=0
      False Float=&MakeCustomFloatFromBool(Bool=$System.IsRunning, True Float=0.5, False Float=1)
     )
    )) {isSetParameters=true}
  ]
  particleSpawnScript [
   &InitializeParticle(
    Color=<rgba 1 1 1 4.22>
    Lifetime=$Emitter.CurrentLoopDuration
    Sprite Size Mode=::NewEnumerator3
    Uniform Sprite Size=20
    UsePositionOffset=false
    Write Material Random=true
   )
   /SetVariables(Particles.DriftPosition:NiagaraPosition=&TransformPosition(Position=<5 0 0>, Source Space=::Local, Destination Space=::World)) {isSetParameters=true}
  ]
  particleUpdateScript [
   &ParticleState(
    Kill Particles When Lifetime Has Elapsed=false
    Let Infinitely Lived Particles Die When Emitter Deactivates=true
    Loop Particles Lifetime=true
   )
   /SetVariables(Particles.DriftPosition:NiagaraPosition=/Niagara/DynamicInputs/Lerp/SmoothLerpOverTime_Position(
     Convergence Rate:NiagaraFloat=&MakeCustomFloatFromBool(Bool=$User.IsSkydiving, True Float=100, False Float=11)
     Smooth Value:NiagaraPosition=$Particles.DriftPosition
     Target Value:NiagaraPosition=&AddVectorToPosition(
      Position=&AddVectorToPosition(
       Position=&TransformPosition(
        Position=&Multiply_Vector(
         A=&MakeVector(
          X=&Sine(Normalized Angle=$Emitter.Age, Period=8, Scale=1, Bias=1)
          Y=&Cosine(Normalized Angle=$Emitter.Age, Period=10, Scale=1, Bias=0)
          Z=&Cosine(Normalized Angle=$Emitter.Age, Period=10, Scale=1, Bias=0)
         )
         B=<0 0 0>
        )
        Source Space=::Local
        Destination Space=::World
       )
       Vector=&MakeCustomVectorFromBool(
        Bool=$User.IsSkydiving
        True Vector=<0 0 0>
        False Vector=&TransformVector(Vector=$Emitter.OffsetTarget, Destination Space=::World)
       )
      )
      Vector=<0 0 0>
     )
    )) {isSetParameters=true}
   /SetVariables(Particles.TransformedPos:NiagaraPosition=&TransformPosition(Position=$Particles.DriftPosition, Source Space=::World, Destination Space=::Simulation)) {isSetParameters=true}
   /SetVariables(Particles.DeltaLag:Vector3f=&SubtractVectorFromPosition(Position=$Particles.DriftPosition, Vector=Linked() {linkedVariable {name=Engine.Owner.Position, type=NiagaraPosition}})) {isSetParameters=true}
   &SolveForcesAndVelocity(Clamp Velocity=false, Limit Acceleration=false)
   &DynamicMaterialParameters(
    Index 0 Param 1=&MakeFloatFromVector(Vector=$Particles.DeltaLag, Channel=::NewEnumerator0)
    Index 0 Param 2=&MakeFloatFromVector(Vector=$Particles.DeltaLag, Channel=::NewEnumerator1)
    Index 0 Param 3=&Max_Float(
     A:NiagaraFloat=&MakeFloatFromVector(Vector=$Particles.DeltaLag, Channel=::NewEnumerator2)
     B:NiagaraFloat=10
    )
    Index 0 Param 4=0.2
    Module.Index0Param0Write=true
    Module.Index0Param1Write=true
    Module.Index0Param2Write=true
    Module.Index0Param3Write=true
   )
   &CameraOffset(Camera Offset Amount=2)
   &Color() {enabled=false}
   &ScaleColor(
    Scale Alpha=&Multiply_Float(
     A=$Emitter.AlphaTarget
     B=&MakeCustomFloatFromBool(Bool=$User.bIsFrontEndPreview, True Float=0.65, False Float=1)
    )
    ScaleA=true
    ScaleRGB=true
   )
  ]
  rendererClasses [&NiagaraMeshRendererProperties, &NiagaraSpriteRendererProperties]
  renderers [
   {rendererClass=&NiagaraMeshRendererProperties, rendererIndex=0}
   {rendererClass=&NiagaraSpriteRendererProperties, rendererIndex=1}
   {rendererClass=&NiagaraSpriteRendererProperties, rendererIndex=2}
   {rendererClass=&NiagaraMeshRendererProperties, rendererIndex=3}
  ]
 }
 SyrupDrips {
  bEnabled=false
  emitterUpdateScript [
   &EmitterState(
    Life Cycle Mode=::NewEnumerator1
    Loop Delay=&RandomRangeFloat(Minimum=0, Maximum=0.222)
    Loop Duration=&RandomRangeFloat(Minimum=1, Maximum=2)
    Recalculate Duration Each Loop=true
    UseLoopDelay=true
   )
   &SpawnRate(
    Spawn Probability=&MakeCustomFloatFromBool(Bool=$System.IsRunning, True Float=1, False Float=0.5)
    SpawnRate=&MakeCustomFloatFromBool(
     Bool=$User.bIsFrontEndPreview
     True Float=2
     False Float=&MakeCustomFloatFromBool(
      Bool=$User.IsSkydiving
      True Float=0
      False Float=&MakeCustomFloatFromBool(Bool=$System.IsRunning, True Float=1.25, False Float=0.5)
     )
    )
    Use Spawn Probability=true
   )
  ]
  particleSpawnScript [
   &PartitionParticles()
   /SetVariables(Particles.BackpackSelectedBone:NiagaraInt32=$Particles.PartitionParticles.Partition) {isSetParameters=true, module=SetVariables_#4}
   &InitializeParticle(
    Color=<rgba 1 1 1 10>
    Lifetime Max=4.3
    Lifetime Min=3.6
    Lifetime Mode=::NewEnumerator1
    Mesh Scale Mode=::NewEnumerator3
    Mesh Uniform Scale=4
    UsePositionOffset=false
    Write Material Random=true
   )
   &SkeletalMeshLocation(
    Bone / Socket Index=$Particles.BackpackSelectedBone
    MaskSpawnGroup=false
    Mesh Sampling Type=::NewEnumerator5
    Skeletal Mesh=$System.Skeletal Mesh
    Socket Sampling Mode=::NewEnumerator5
   )
   /SetVariables(Particles.Scale=&Multiply_VectorByFloat(
     Vector=&VectorFromFloat(Value=&MakeCustomFloatFromBool(Bool=$User.bIsFrontEndPreview, True Float=0.075, False Float=0.075))
     Float=&SelectFloatFromArray(Direct Array Index=$Particles.BackpackSelectedBone, Float Selection Array=DataInterface() {propertyValues {FloatData [0.7,0.75,0.65,0.62]}})
    )) {isSetParameters=true, module=SetVariables_#5}
   &CameraOffset(Camera Offset Amount=&MakeCustomFloatFromBool(Bool=$User.bIsFrontEndPreview, True Float=-1, False Float=0))
   &InitialMeshOrientation(Mesh Orientation Mode=::NewEnumerator2, Orientation Vector=<-1 0 0>, Use Rotation Vector=false) {enabled=false}
   &DynamicMaterialParameters(
    Index 0 Param 3=&FN_FrontEndFloatModifier(Gameplay=0, Front End=1)
    Module.Index0Param0Write=true
    Module.Index0Param1Write=true
    Module.Index0Param2Write=true
    Module.Index0Param3Write=false
   )
  ]
  particleUpdateScript [
   /SetVariables(Particles.Connection:NiagaraBool=&SetBoolByFloatComparison(A=$Particles.NormalizedAge, B=0.85)) {isSetParameters=true, module=SetVariables_#6}
   &ScaleColor(
    Scale Alpha=&FloatFromCurve(FloatCurve=DataInterface() {propertyValues {Curve {keys [{interpMode=RCIM_Cubic, tangentMode=RCTM_User, time=0.2, value=0}, {interpMode=RCIM_Cubic, tangentMode=RCTM_User, time=0.45, value=1}]}}})
    Scale RGB=&FN_FrontEndVector3Modifier(
     Gameplay=&MakeVectorFromLinearColorRGB(Color=&FNTimeOfDayColor(
       Night Color=<rgba 2.8 2.128 0.28 1>
       Morning Color=<rgba 0.9 0.6345 0.369 1>
       Day Color=<rgba 0.65 0.4647 0.065 1>
       Evening Color=<rgba 1.1 0.8784 0.077 1>
      ))
     Front End=&MakeVectorFromLinearColorRGB(Color=<rgba 0.26 0.2436 0.1196 1>)
    )
    ScaleA=true
    ScaleRGB=true
   ) {enabled=false}
   &ScaleColor(
    Scale Alpha=&FloatFromCurve(FloatCurve=DataInterface() {propertyValues {Curve {keys [{interpMode=RCIM_Cubic, tangentMode=RCTM_User, time=0.2, value=0}, {interpMode=RCIM_Cubic, tangentMode=RCTM_User, time=0.45, value=1}]}}})
    Scale RGB=&FN_FrontEndVector3Modifier(
     Gameplay=&MakeVectorFromLinearColorRGB(Color=&FNTimeOfDayColor(
       Night Color=<rgba 1 0.439 0.01 1>
       Morning Color=<rgba 0.2 0.08 0.02 1>
       Day Color=<rgba 0.14 0.0711 0.0252 1>
       Evening Color=<rgba 0.27 0.0999 0.027 1>
      ))
     Front End=&MakeVectorFromLinearColorRGB(Color=<rgba 0.26 0.2436 0.1196 1>)
    )
    ScaleA=true
    ScaleRGB=true
   ) {enabled=false, module=ScaleColor001}
   /SetVariables(Particles.Color=&MultiplyLinearColor(
     Linear Color=&MakeCustomLinearColorFromBool(Bool=$User.bIsFrontEndPreview, True Color=<rgba 0.7552 0.1783 0 1>, False Color=<rgba 0.8438 0.3095 0 1>)
     Scale Factor=&MakeVector4_FromLinearColor(LinearColor=&MakeLinearColorFromVectorAndFloat(
       "Vector (RGB)"=&FN_FrontEndVector3Modifier(
        Gameplay=&MakeVectorFromLinearColorRGB(Color=&FNTimeOfDayColor(
          Night Color=<rgba 1 1 1 1>
          Morning Color=<rgba 0.7031 0.7031 0.7031 1>
          Day Color=<rgba 0.7031 0.7031 0.7031 1>
          Evening Color=<rgba 1 1 1 1>
         ))
        Front End=&MakeVectorFromLinearColorRGB(Color=<rgba 0.7031 0.7031 0.7031 1>)
       )
       "Float (Alpha)"=&FloatFromCurve(
        FloatCurve=DataInterface() {propertyValues {Curve {keys [{interpMode=RCIM_Cubic, tangentMode=RCTM_User, time=0.2, value=0}, {interpMode=RCIM_Cubic, tangentMode=RCTM_User, time=0.45, value=1}]}}}
        Scale Curve=&FN_FrontEndFloatModifier(Gameplay=&FNTimeOfDayScalar(Night Scale=1, Morning Scale=1, Evening Scale=1), Front End=0.27)
       )
      ))
    )) {isSetParameters=true, module=SetVariables_#7}
   &ParticleState()
   &ScaleMeshSize(Scale Factor=&VectorFromFloat(Value=&FloatFromCurve(FloatCurve=DataInterface() {propertyValues {Curve {keys [
          {arriveTangent=0.6088, interpMode=RCIM_Cubic, leaveTangent=0.6088, tangentMode=RCTM_User, time=0.15, value=0.4}
          {interpMode=RCIM_Cubic, tangentMode=RCTM_User, time=0.7, value=1}
          {arriveTangent=-0.0171, interpMode=RCIM_Cubic, leaveTangent=-0.0171, tangentMode=RCTM_User, time=0.95, value=1}
          {arriveTangent=-22.3504, interpMode=RCIM_Cubic, leaveTangent=-22.3504, tangentMode=RCTM_User, time=1, value=0}
         ]}}})))
   &ScaleMeshSize(Scale Factor=&MakeVector(
     X=1
     Y=1
     Z=&FloatFromCurve(
      FloatCurve=DataInterface() {propertyValues {Curve {keys [
          {arriveTangent=4.2114, interpMode=RCIM_Cubic, leaveTangent=4.2114, tangentMode=RCTM_User, time=0.2, value=1}
          {arriveTangent=0.5863, interpMode=RCIM_Cubic, leaveTangent=0.5863, tangentMode=RCTM_User, time=1, value=2.5}
         ]}}}
      Scale Curve=&SelectFloatFromArray(Direct Array Index=$Particles.BackpackSelectedBone, Float Selection Array=DataInterface() {propertyValues {FloatData [1,1,1.5,0.8]}})
     )
    )) {module=ScaleMeshSize001}
   &SampleSkeletalMesh(
    AdditionalYawPitchRoll=false
    Bone / Socket Index:NiagaraInt32=$Particles.BackpackSelectedBone
    Mesh Sampling Type:ENiagaraSkelSamplingModeFull=::NewEnumerator5
    Skeletal Mesh:NiagaraDataInterfaceSkeletalMesh=$System.Skeletal Mesh
    Socket Sampling Mode:ENiagaraSocketSamplingMode=::NewEnumerator5
   )
   /SetVariables(Particles.MeshOrientation=&FindQuatBetweenTwoVectors(
     From Vector=<0 0 -1.0499>
     To Vector=&MakeCustomVectorFromBool(
      Bool=$Particles.Connection
      True Vector=&NormalizeVector(
       Vector To Normalize=&SubtractVectorFromPosition(Position=$Particles.Position, Vector=$Particles.SampleSkeletalMesh.SampledPosition)
       Safe Normalize=true
       Safe Fallback Threshold=0
       Safe Fallback Vector=<0 0 0.001>
      )
      False Vector=&NormalizeVector(
       Vector To Normalize=&Lerp_Vector(
        A=&SubtractVectorFromPosition(Position=$Particles.Position, Vector=$Particles.SampleSkeletalMesh.SampledPosition)
        B=<0 0 -1>
        Alpha=&FloatFromCurve(FloatCurve=DataInterface() {propertyValues {Curve {keys [<at 0.7 0>,<at 0.8 1>]}}}, Scale Curve=0)
       )
       Safe Normalize=true
       Safe Fallback Threshold=0
       Safe Fallback Vector=<0 0 0.001>
      )
     )
    )) {isSetParameters=true, module=SetVariables_#8}
   &GravityForce()
   &Drag(Drag=2, Use Linear Drag=true, Use Rotational Drag=true)
   &PendulumSetup(
    Module.Enable Spring Driver=false
    Pendulum Length:NiagaraFloat=&FloatFromCurve(
     FloatCurve=DataInterface() {propertyValues {Curve {keys [<at 0 0 RCIM_Cubic>, {arriveTangent=0.0714, interpMode=RCIM_Cubic, leaveTangent=0.0714, tangentMode=RCTM_User, time=0.7, value=1}]}}}
     Scale Curve=&SelectFloatFromArray(Direct Array Index=$Particles.BackpackSelectedBone, Float Selection Array=DataInterface() {propertyValues {FloatData [0.1,0.1,0.1,0.1]}})
    )
    Pendulum Pivot:Vector3f=$Particles.SampleSkeletalMesh.SampledPosition
    Rigid Pendulum=false
    "Spring Driven Constraint (Tightness)"=2
   )
   &AccelerationForce(Acceleration=&Multiply_VectorByFloat(
     Vector=&Multiply_Vector(A=$Engine.Owner.Velocity, B=<-1 -1 0>)
     Float=&FloatFromCurve(FloatCurve=DataInterface() {propertyValues {Curve {keys [{interpMode=RCIM_Cubic, tangentMode=RCTM_User, time=0.6, value=0}, {interpMode=RCIM_Cubic, tangentMode=RCTM_User, time=0.9, value=1}]}}})
    ))
   &SolveForcesAndVelocity(Clamp Velocity=false, Limit Acceleration=false)
   /SetVariables(Particles.PositionTemp:NiagaraPosition=$Particles.Position) {isSetParameters=true, module=SetVariables_#9}
   &ScaleMeshSizeBySpeed(Max Scale Factor=<1 1 0.3>, Min Scale Factor:Vector3f=<0.6 0.6 1>, Speed Threshold=300)
   &PendulumConstraint()
   /SetVariables(Particles.Position=&MakeCustomVectorFromBool(
     Bool=$Particles.Connection
     True Vector=Linked() {linkedVariable {name=Particles.PositionTemp, type=NiagaraPosition}}
     False Vector=Linked() {linkedVariable {name=Particles.Position, type=NiagaraPosition}}
    )) {isSetParameters=true, module=SetVariables_#10}
   &CalculateAccurateVelocity()
   /SetVariables(Particles.MeshOrientation=&LerpQuaternion(
     Quaternion A:Quat4f=$StackContext.MeshOrientation
     Quaternion B:Quat4f=&MultiplyQuaternion(Quaternion A=$StackContext.SampleSkeletalMesh.SampledRotation, Quaternion B=<0 0 0 1>)
     Lerp Factor:NiagaraFloat=1
     Enable Spherical Linear Interpolation:NiagaraBool=false
    )) {isSetParameters=true, module=SetVariables_#11}
   /SetVariables(Particles.PositionDrips:NiagaraPosition=&AddVectorToPosition(
     Position=$Particles.Position
     Vector=&Multiply_VectorByFloat(Vector=&RotateVectorByQuaternion(Vector To Rotate=<0 0 1>, Quaternion=$Particles.MeshOrientation), Float=0)
    )) {isSetParameters=true, module=SetVariables_#12}
  ]
  rendererClasses [&NiagaraMeshRendererProperties]
  renderers [
   {rendererClass=&NiagaraMeshRendererProperties, rendererIndex=0}
   {rendererClass=&NiagaraMeshRendererProperties, rendererIndex=1}
  ]
 }
 SyrupDrips001 {
  emitterUpdateScript [
   &EmitterState(
    Life Cycle Mode=::NewEnumerator1
    Loop Delay=&RandomRangeFloat(Minimum=0, Maximum=0.222)
    Loop Duration=&RandomRangeFloat(Minimum=1, Maximum=2)
    Recalculate Duration Each Loop=true
    UseLoopDelay=true
   )
   &SpawnRate(
    Spawn Probability=&MakeCustomFloatFromBool(Bool=$System.IsRunning, True Float=1, False Float=0.5)
    SpawnRate=&MakeCustomFloatFromBool(
     Bool=$User.bIsFrontEndPreview
     True Float=2
     False Float=&MakeCustomFloatFromBool(
      Bool=$User.IsSkydiving
      True Float=0
      False Float=&MakeCustomFloatFromBool(Bool=$System.IsRunning, True Float=1.25, False Float=0.5)
     )
    )
    Use Spawn Probability=true
   )
  ]
  particleSpawnScript [
   &PartitionParticles()
   /SetVariables(Particles.BackpackSelectedBone:NiagaraInt32=$Particles.PartitionParticles.Partition) {isSetParameters=true, module=SetVariables_#4}
   &InitializeParticle(
    Color=<rgba 1 1 1 10>
    Lifetime Max=3.3
    Lifetime Min=2.5
    Lifetime Mode=::NewEnumerator1
    Mesh Scale Mode=::NewEnumerator3
    Mesh Uniform Scale=4
    UsePositionOffset=false
    Write Material Random=true
   )
   &SkeletalMeshLocation(
    Bone / Socket Index=$Particles.BackpackSelectedBone
    MaskSpawnGroup=false
    Mesh Sampling Type=::NewEnumerator5
    Skeletal Mesh=$System.Skeletal Mesh
    Socket Sampling Mode=::NewEnumerator5
   )
   /SetVariables(Particles.Scale=&Multiply_VectorByFloat(
     Vector=&VectorFromFloat(Value=&MakeCustomFloatFromBool(Bool=$User.bIsFrontEndPreview, True Float=0.075, False Float=0.095))
     Float=&SelectFloatFromArray(Direct Array Index=$Particles.BackpackSelectedBone, Float Selection Array=DataInterface() {propertyValues {FloatData [0.7,0.75,0.65,0.62]}})
    )) {isSetParameters=true, module=SetVariables_#5}
   &CameraOffset(Camera Offset Amount=&MakeCustomFloatFromBool(Bool=$User.bIsFrontEndPreview, True Float=-1, False Float=0))
   &InitialMeshOrientation(Mesh Orientation Mode=::NewEnumerator2, Orientation Vector=<-1 0 0>, Use Rotation Vector=false) {enabled=false}
   &DynamicMaterialParameters(
    Index 0 Param 3=&FN_FrontEndFloatModifier(Gameplay=0, Front End=1)
    Module.Index0Param0Write=true
    Module.Index0Param1Write=true
    Module.Index0Param2Write=true
    Module.Index0Param3Write=false
   )
  ]
  particleUpdateScript [
   /SetVariables(Particles.Connection:NiagaraBool=&SetBoolByFloatComparison(A=$Particles.NormalizedAge, B=0.85)) {isSetParameters=true, module=SetVariables_#6}
   &ScaleColor(
    Scale Alpha=&FloatFromCurve(FloatCurve=DataInterface() {propertyValues {Curve {keys [{interpMode=RCIM_Cubic, tangentMode=RCTM_User, time=0.2, value=0}, {interpMode=RCIM_Cubic, tangentMode=RCTM_User, time=0.45, value=1}]}}})
    Scale RGB=&FN_FrontEndVector3Modifier(
     Gameplay=&MakeVectorFromLinearColorRGB(Color=&FNTimeOfDayColor(
       Night Color=<rgba 2.8 2.128 0.28 1>
       Morning Color=<rgba 0.9 0.6345 0.369 1>
       Day Color=<rgba 0.65 0.4647 0.065 1>
       Evening Color=<rgba 1.1 0.8784 0.077 1>
      ))
     Front End=&MakeVectorFromLinearColorRGB(Color=<rgba 0.26 0.2436 0.1196 1>)
    )
    ScaleA=true
    ScaleRGB=true
   ) {enabled=false}
   &ScaleColor(
    Scale Alpha=&FloatFromCurve(FloatCurve=DataInterface() {propertyValues {Curve {keys [{interpMode=RCIM_Cubic, tangentMode=RCTM_User, time=0.2, value=0}, {interpMode=RCIM_Cubic, tangentMode=RCTM_User, time=0.45, value=1}]}}})
    Scale RGB=&FN_FrontEndVector3Modifier(
     Gameplay=&MakeVectorFromLinearColorRGB(Color=&FNTimeOfDayColor(
       Night Color=<rgba 1 0.439 0.01 1>
       Morning Color=<rgba 0.2 0.08 0.02 1>
       Day Color=<rgba 0.14 0.0711 0.0252 1>
       Evening Color=<rgba 0.27 0.0999 0.027 1>
      ))
     Front End=&MakeVectorFromLinearColorRGB(Color=<rgba 0.26 0.2436 0.1196 1>)
    )
    ScaleA=true
    ScaleRGB=true
   ) {enabled=false, module=ScaleColor001}
   /SetVariables(Particles.Color=&MultiplyLinearColor(
     Linear Color=&MakeCustomLinearColorFromBool(Bool=$User.bIsFrontEndPreview, True Color=<rgba 0.7552 0.1783 0 1>, False Color=<rgba 0.8438 0.3095 0 1>)
     Scale Factor=&MakeVector4_FromLinearColor(LinearColor=&MakeLinearColorFromVectorAndFloat(
       "Vector (RGB)"=&FN_FrontEndVector3Modifier(
        Gameplay=&MakeVectorFromLinearColorRGB(Color=&FNTimeOfDayColor(
          Night Color=<rgba 1 1 1 1>
          Morning Color=<rgba 0.7031 0.7031 0.7031 1>
          Day Color=<rgba 0.7031 0.7031 0.7031 1>
          Evening Color=<rgba 1 1 1 1>
         ))
        Front End=&MakeVectorFromLinearColorRGB(Color=<rgba 0.7031 0.7031 0.7031 1>)
       )
       "Float (Alpha)"=&FloatFromCurve(
        FloatCurve=DataInterface() {propertyValues {Curve {keys [{interpMode=RCIM_Cubic, tangentMode=RCTM_User, time=0.2, value=0}, {interpMode=RCIM_Cubic, tangentMode=RCTM_User, time=0.45, value=1}]}}}
        Scale Curve=&FN_FrontEndFloatModifier(Gameplay=&FNTimeOfDayScalar(Night Scale=1, Morning Scale=1, Evening Scale=1), Front End=0.27)
       )
      ))
    )) {isSetParameters=true, module=SetVariables_#7}
   &ParticleState()
   &ScaleMeshSize(Scale Factor=&VectorFromFloat(Value=&FloatFromCurve(FloatCurve=DataInterface() {propertyValues {Curve {keys [
          {arriveTangent=0.6088, interpMode=RCIM_Cubic, leaveTangent=0.6088, tangentMode=RCTM_User, time=0.15, value=0.4}
          {interpMode=RCIM_Cubic, tangentMode=RCTM_User, time=0.7, value=1}
          {arriveTangent=-0.0171, interpMode=RCIM_Cubic, leaveTangent=-0.0171, tangentMode=RCTM_User, time=0.95, value=1}
          {arriveTangent=-22.3504, interpMode=RCIM_Cubic, leaveTangent=-22.3504, tangentMode=RCTM_User, time=1, value=0}
         ]}}})))
   &ScaleMeshSize(Scale Factor=&MakeVector(
     X=1
     Y=1
     Z=&FloatFromCurve(
      FloatCurve=DataInterface() {propertyValues {Curve {keys [
          {arriveTangent=4.2114, interpMode=RCIM_Cubic, leaveTangent=4.2114, tangentMode=RCTM_User, time=0.2, value=1}
          {arriveTangent=0.5863, interpMode=RCIM_Cubic, leaveTangent=0.5863, tangentMode=RCTM_User, time=1, value=2.5}
         ]}}}
      Scale Curve=&SelectFloatFromArray(Direct Array Index=$Particles.BackpackSelectedBone, Float Selection Array=DataInterface() {propertyValues {FloatData [1,1,1.5,0.8]}})
     )
    )) {module=ScaleMeshSize001}
   &SampleSkeletalMesh(
    AdditionalYawPitchRoll=false
    Bone / Socket Index:NiagaraInt32=$Particles.BackpackSelectedBone
    Mesh Sampling Type:ENiagaraSkelSamplingModeFull=::NewEnumerator5
    Skeletal Mesh:NiagaraDataInterfaceSkeletalMesh=$System.Skeletal Mesh
    Socket Sampling Mode:ENiagaraSocketSamplingMode=::NewEnumerator5
   )
   /SetVariables(Particles.MeshOrientation=&FindQuatBetweenTwoVectors(
     From Vector=<0 0 -1.0499>
     To Vector=&MakeCustomVectorFromBool(
      Bool=$Particles.Connection
      True Vector=&NormalizeVector(
       Vector To Normalize=&SubtractVectorFromPosition(Position=$Particles.Position, Vector=$Particles.SampleSkeletalMesh.SampledPosition)
       Safe Normalize=true
       Safe Fallback Threshold=0
       Safe Fallback Vector=<0 0 0.001>
      )
      False Vector=&NormalizeVector(
       Vector To Normalize=&Lerp_Vector(
        A=&SubtractVectorFromPosition(Position=$Particles.Position, Vector=$Particles.SampleSkeletalMesh.SampledPosition)
        B=<0 0 -1>
        Alpha=&FloatFromCurve(FloatCurve=DataInterface() {propertyValues {Curve {keys [<at 0.7 0>,<at 0.8 1>]}}}, Scale Curve=0)
       )
       Safe Normalize=true
       Safe Fallback Threshold=0
       Safe Fallback Vector=<0 0 0.001>
      )
     )
    )) {isSetParameters=true, module=SetVariables_#8}
   &GravityForce()
   &Drag(Drag=2, Use Linear Drag=true, Use Rotational Drag=true)
   &PendulumSetup(
    Module.Enable Spring Driver=false
    Pendulum Length:NiagaraFloat=&FloatFromCurve(
     FloatCurve=DataInterface() {propertyValues {Curve {keys [<at 0 0 RCIM_Cubic>, {arriveTangent=0.0714, interpMode=RCIM_Cubic, leaveTangent=0.0714, tangentMode=RCTM_User, time=0.7, value=1}]}}}
     Scale Curve=3
    )
    Pendulum Pivot:Vector3f=$Particles.SampleSkeletalMesh.SampledPosition
    "Spring Driven Constraint (Tightness)"=2
   )
   &AccelerationForce(Acceleration=&Multiply_VectorByFloat(
     Vector=&Multiply_Vector(A=$Engine.Owner.Velocity, B=<-1 -1 0>)
     Float=&FloatFromCurve(FloatCurve=DataInterface() {propertyValues {Curve {keys [{interpMode=RCIM_Cubic, tangentMode=RCTM_User, time=0.6, value=0}, {interpMode=RCIM_Cubic, tangentMode=RCTM_User, time=0.9, value=1}]}}})
    ))
   &SolveForcesAndVelocity(Clamp Velocity=false, Limit Acceleration=false)
   /SetVariables(Particles.PositionTemp:NiagaraPosition=$Particles.Position) {isSetParameters=true, module=SetVariables_#9}
   &ScaleMeshSizeBySpeed(Max Scale Factor=<1 1 0.3>, Min Scale Factor:Vector3f=<0.6 0.6 1>, Speed Threshold=300)
   &PendulumConstraint()
   /SetVariables(Particles.Position=&MakeCustomVectorFromBool(
     Bool=$Particles.Connection
     True Vector=Linked() {linkedVariable {name=Particles.PositionTemp, type=NiagaraPosition}}
     False Vector=Linked() {linkedVariable {name=Particles.Position, type=NiagaraPosition}}
    )) {isSetParameters=true, module=SetVariables_#10}
   &CalculateAccurateVelocity()
   /SetVariables(Particles.MeshOrientation=&LerpQuaternion(
     Quaternion A:Quat4f=$StackContext.MeshOrientation
     Quaternion B:Quat4f=&MultiplyQuaternion(Quaternion A=$StackContext.SampleSkeletalMesh.SampledRotation, Quaternion B=<0 0 0 1>)
     Lerp Factor:NiagaraFloat=1
     Enable Spherical Linear Interpolation:NiagaraBool=false
    )) {enabled=false, isSetParameters=true, module=SetVariables_#11}
   /SetVariables(Particles.PositionDrips:NiagaraPosition=&MakeCustomVectorFromBool(
     Bool=&SetBoolByFloatComparison(A=$Particles.NormalizedAge, B=0.85)
     True Vector=Linked() {linkedVariable {name=Particles.Position, type=NiagaraPosition}}
     False Vector=$Transient.PhysicsPendulumPivot
    )) {isSetParameters=true, module=SetVariables_#12}
  ]
  rendererClasses [&NiagaraMeshRendererProperties]
  renderers [
   {rendererClass=&NiagaraMeshRendererProperties, rendererIndex=0}
   {rendererClass=&NiagaraMeshRendererProperties, rendererIndex=1}
  ]
 }
}
summary {
 systemName=NS_Backpack_PeacefulPoemRemix_Idle
 emitters [
  {bEnabled=false, emitterName=Droplets, rendererClasses [&NiagaraSpriteRendererProperties]}
  {emitterName=MeshSteam, rendererClasses [&NiagaraMeshRendererProperties, &NiagaraSpriteRendererProperties]}
  {bEnabled=false, emitterName=SyrupDrips, rendererClasses [&NiagaraMeshRendererProperties]}
  {emitterName=SyrupDrips001, rendererClasses [&NiagaraMeshRendererProperties]}
 ]
 userVariables [
  User.bIsFrontEnd {type=NiagaraBool}
  User.bIsFrontEndPreview {type=NiagaraBool}
  User.IsSkydiving {type=NiagaraBool}
 ]
}
