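# Physics asset for the PeacefulPoemRemix backpack cosmetic, defining secondary motion for its dangling box, pizza, cheese, star and dagger pieces.
# Capsule bodies sit on the dyn joints, with two boxes forming the pack body.
# Box joints 1 and 3 hinge narrowly off the pack; box 1 carries box 2, stars and dagger, box 3 the looser pizza and cheese chains.
# Box joints 4–6 have large capsules but no constraints.
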PhysicsAsset /BRCosmetics/Accessories/FORT_Backpacks/Backpack_M_MED_PeacefulPoemRemix/Meshes/M_MED_PeacefulPoemRemix_Pack_Physics
bodies {
 dyn_box_1_joint [{shapeType=Capsule, length=19, radius=2, shapeName=dyn_box_1_joint_capsule, center=<-0.5008 0 0>, rotation=<rot -90 90 -89.999>}]
 dyn_box_2_joint [{shapeType=Capsule, length=12, radius=3, shapeName=dyn_box_2_joint_capsule, center=<4.474 1.251 0>, rotation=<rot 90 179.5524 176.9256>}]
 dyn_box_3_joint [{shapeType=Capsule, length=3.5, radius=2, shapeName=dyn_box_3_joint_capsule, center=<0 0 -8.4318>}]
 dyn_box_4_joint [{shapeType=Capsule, length=18.5, radius=12.5, shapeName=dyn_box_4_joint_capsule, rotation=<rot 0 179.9995 -179.9998>}]
 dyn_box_5_joint [{shapeType=Capsule, length=18.5, radius=12.5, shapeName=dyn_box_5_joint_capsule, rotation=<rot 0 179.9995 -179.9998>}]
 dyn_box_6_joint [{shapeType=Capsule, length=18.5, radius=12.5, shapeName=dyn_box_6_joint_capsule, rotation=<rot 0 179.9995 -179.9998>}]
 dyn_cheese_l_1_joint [{shapeType=Capsule, length=0.5, radius=0.5, shapeName=dyn_cheese_l_1_joint_capsule, center=<0.4427 0 0>, rotation=<rot -90 -0.0001 0.0003>}]
 dyn_cheese_l_2_joint [{shapeType=Capsule, length=0.5, radius=0.5, shapeName=dyn_cheese_l_2_joint_capsule, center=<1.655 0 0>, rotation=<rot 90 -179.9602 -179.9607>}]
 dyn_cheese_l_3_joint [{shapeType=Capsule, length=0.5, radius=0.5, shapeName=dyn_cheese_l_3_joint_capsule, center=<1.8694 0 0>, rotation=<rot 90 180 179.9995>}]
 dyn_cheese_m_1_joint [{shapeType=Capsule, length=0, radius=0.5, shapeName=dyn_cheese_m_1_joint_capsule, center=<0.5412 0 0>, rotation=<rot -90 180 -179.9994>}]
 dyn_cheese_m_2_joint [{shapeType=Capsule, length=0.5, radius=0.5, shapeName=dyn_cheese_m_2_joint_capsule, center=<1.5332 0 0>, rotation=<rot -90 0 0.0004>}]
 dyn_cheese_m_3_joint [{shapeType=Capsule, length=0.5, radius=0.5, shapeName=dyn_cheese_m_3_joint_capsule, center=<1.1306 0 0>, rotation=<rot -90 180 -179.9996>}]
 dyn_cheese_r_joint [{shapeType=Capsule, length=0.5, radius=0.5, shapeName=dyn_cheese_r_joint_capsule, center=<1.0893 0 0>, rotation=<rot -90 180 -179.9993>}]
 dyn_dagger_joint [{shapeType=Capsule, length=10, radius=0.5, shapeName=dyn_dagger_joint_capsule, center=<-7.9669 -0.0001 0>, rotation=<rot -90 -0.0009 0.0013>}]
 dyn_pizza_1_joint [{shapeType=Capsule, length=3.5, radius=1, shapeName=dyn_pizza_1_joint_capsule, center=<3.6526 0 0>, rotation=<rot 90 179.9607 179.9602>}]
 dyn_pizza_2_joint [{shapeType=Capsule, length=2, radius=1, shapeName=dyn_pizza_2_joint_capsule, center=<1.5311 0 0>, rotation=<rot 90 180 179.9998>}]
 dyn_pizza_3_joint [{shapeType=Capsule, length=1, radius=1, shapeName=dyn_pizza_3_joint_capsule, center=<1.693 0 0>, rotation=<rot 90 180 179.9998>}]
 dyn_pizza_4_joint [{shapeType=Capsule, length=0.5, radius=0.5, shapeName=dyn_pizza_4_joint_capsule, center=<0.9472 0 0>, rotation=<rot 90 177.4696 177.4691>}]
 dyn_stars_down_joint [{
   shapeType=Capsule
   length=10.5
   radius=1
   shapeName=dyn_stars_down_joint_capsule
   center=<4.0154 0 -0.3513>
   rotation=<rot -85 -179.9992 179.9995>
  }]
 dyn_stars_up_joint [{shapeType=Capsule, length=11, radius=1, shapeName=dyn_stars_up_joint_capsule, center=<3.0446 0 0>, rotation=<rot 90 90 89.9995>}]
 pack [
  {shapeType=Box, extentX=27.5, extentY=1.5, extentZ=26.5, shapeName=pack_box, center=<0 0.4187 -4.3357>}
  {
   shapeType=Box
   extentX=1.5
   extentY=5.5
   extentZ=27.5
   shapeName=pack_box
   center=<-0.0375 -1.614 -16.8467>
   rotation=<rot 90 -179.9901 -179.9902>
  }
 ]
}
constraints [
 {bone1Name=dyn_pizza_2_joint, bone2Name=dyn_pizza_1_joint, swing1LimitDegrees=35, swing2LimitDegrees=5, twistLimitDegrees=1}
 {bone1Name=dyn_pizza_3_joint, bone2Name=dyn_pizza_2_joint, swing1LimitDegrees=50, swing2LimitDegrees=5, twistLimitDegrees=1}
 {bone1Name=dyn_pizza_4_joint, bone2Name=dyn_pizza_3_joint, swing1LimitDegrees=50, swing2LimitDegrees=10, twistLimitDegrees=1}
 {
  bone1Name=dyn_box_2_joint
  bone2Name=dyn_box_1_joint
  swing1LimitDegrees=5
  swing2LimitDegrees=0
  swing2Motion=Locked
  twistLimitDegrees=0
  twistMotion=Locked
 }
 {bone1Name=dyn_stars_up_joint, bone2Name=dyn_box_1_joint, swing1LimitDegrees=7, swing2LimitDegrees=12, twistLimitDegrees=3}
 {bone1Name=dyn_stars_down_joint, bone2Name=dyn_box_1_joint, swing1LimitDegrees=7, swing2LimitDegrees=12, twistLimitDegrees=3}
 {bone1Name=dyn_pizza_1_joint, bone2Name=dyn_box_3_joint, swing1LimitDegrees=70, swing2LimitDegrees=5, twistLimitDegrees=1}
 {bone1Name=dyn_cheese_r_joint, bone2Name=dyn_box_3_joint, swing1LimitDegrees=50, swing2LimitDegrees=50, twistLimitDegrees=1}
 {bone1Name=dyn_dagger_joint, bone2Name=dyn_box_1_joint, swing1LimitDegrees=10, swing2LimitDegrees=15, twistLimitDegrees=3}
 {bone1Name=dyn_cheese_l_1_joint, bone2Name=dyn_box_3_joint, swing1LimitDegrees=50, swing2LimitDegrees=50, twistLimitDegrees=1}
 {bone1Name=dyn_cheese_l_2_joint, bone2Name=dyn_cheese_l_1_joint, swing1LimitDegrees=50, swing2LimitDegrees=50, twistLimitDegrees=1}
 {bone1Name=dyn_cheese_l_3_joint, bone2Name=dyn_cheese_l_2_joint, swing1LimitDegrees=50, swing2LimitDegrees=50, twistLimitDegrees=1}
 {bone1Name=dyn_cheese_m_1_joint, bone2Name=dyn_box_3_joint, swing1LimitDegrees=50, swing2LimitDegrees=50, twistLimitDegrees=1}
 {bone1Name=dyn_cheese_m_2_joint, bone2Name=dyn_cheese_m_1_joint, swing1LimitDegrees=50, swing2LimitDegrees=50, twistLimitDegrees=1}
 {bone1Name=dyn_cheese_m_3_joint, bone2Name=dyn_cheese_m_2_joint, swing1LimitDegrees=50, swing2LimitDegrees=50, twistLimitDegrees=1}
 {
  bone1Name=dyn_box_1_joint
  bone2Name=pack
  swing1LimitDegrees=7
  swing2LimitDegrees=0
  swing2Motion=Locked
  twistLimitDegrees=0
  twistMotion=Locked
 }
 {bone1Name=dyn_box_3_joint, bone2Name=pack, swing1LimitDegrees=5, swing2LimitDegrees=0, twistLimitDegrees=0}
]
